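; Gameplay cue notify for PoppySoap that triggers the announcer's "Fight" voice-over.
; Its burst handler runs the parent behavior, then casts the cue target to a Fort player pawn.
; Only for a locally controlled pawn, it adds the announcer Fight dynamic UI scene to that player's controller through the Dynamic UI Manager, so the VO is delivered via a UI scene.
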(fn OnBurst(EitherVFX-preferthisversion)
  (params (MyTarget :type Actor%20Object%20Reference) (Parameters :type Gameplay%20Cue%20Parameters%20Structure%20%28by%20ref%29) (ParticleComponents :type Array%20of%20FXSystem%20Component%20Object%20References) (AudioComponents :type Array%20of%20Audio%20Component%20Object%20References) (BurstCameraShakeInstance :type Camera%20Shake%20Base%20Object%20Reference) (BurstDecalInstance :type Decal%20Actor%20Object%20Reference))
  ;[0.000369 0.15 0.03675 0.5] OnBurst: Fight announcer VO invoked.
  (|CallParentFunction MyTarget Parameters ParticleComponents AudioComponents BurstCameraShakeInstance BurstDecalInstance)
  (Utilities|FlowControl|Sequence
    (:then_0
      (bind _asfort_player_pawn (Utilities|Casting|CastToFortPlayerPawn MyTarget))
      (if (Pawn|IsLocallyControlled _asfort_player_pawn)
        (DynamicUI|AddScene (WorldSubsystems|GetDynamicUIManager _asfort_player_pawn) "/PoppySoap/UI/DynamicUIScene_PoppySoap_AnnouncerVO_Fight.DynamicUIScene_PoppySoap_AnnouncerVO_Fight" (Fortnite|GetFortPlayerControllerfromActor _asfort_player_pawn))))
    (:then_1
      (return))))

(fn ConstructionScript)
